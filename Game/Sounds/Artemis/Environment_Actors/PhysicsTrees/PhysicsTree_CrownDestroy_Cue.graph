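# Sound cue for the destruction of a physics tree's crown, classed as an impact sound with small-object demolition attenuation.
# A random node picks one of three equally weighted crown-destroyed waves and feeds a modulator that slightly varies pitch upward.
# The root quality-level node routes that chain to its first two slots and leaves the other two empty.
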SoundCue /Game/Sounds/Artemis/Environment_Actors/PhysicsTrees/PhysicsTree_CrownDestroy_Cue
root=QualityLevel_0
nodes [
 {id=Modulator_0, children [Random_0], properties {pitchMax=1.1, pitchMin=1}}
 {id=QualityLevel_0, children [Modulator_0,Modulator_0,null,null]}
 {id=Random_0, children [WavePlayer_1,WavePlayer_2,WavePlayer_3], properties {weights [1,1,1]}}
 {id=WavePlayer_1, properties {soundWaveAssetPtr=&/PhysicsTree_Crown_Destroyed_01}}
 {id=WavePlayer_2, properties {soundWaveAssetPtr=&/PhysicsTree_Crown_Destroyed_02}}
 {id=WavePlayer_3, properties {soundWaveAssetPtr=&/PhysicsTree_Crown_Destroyed_03}}
]
properties {
 attenuationSettings=/Game/Sounds/Attenuation/Fort_Demolition_Destroy_Small_Sized_Object_Att
 maxDistance=4000
 soundClassObject=&Fort_Impacts
 volumeMultiplier=0.6
}
