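# Sound cue for a physics tree being destroyed, blending a close layer with a delayed distant layer under a shared pitch modulator.
# The close layer mixes randomized tree-destroyed waves with reused large-tree demolition break sounds, each pitch-varied.
# The distant layer plays a random far variant after a short fixed delay with boosted volume.
# Each layer has its own attenuation and weapon-fire sound class override, and cue-level concurrency caps simultaneous instances.
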SoundCue /Game/Sounds/Artemis/Environment_Actors/PhysicsTrees/PhysicsTree_Destroyed_Cue
root=Modulator_7
nodes [
 {
  id=Attenuation_2
  children [Mixer_0]
  properties {attenuationSettings=/Game/Sounds/Fort_Impact_Sounds/Demolition/Tree/Demo_Tree_Large_Destroyed_Close_Att}
 }
 {
  id=Attenuation_7
  children [Delay_1]
  properties {attenuationSettings=/Game/Sounds/Fort_Impact_Sounds/Demolition/Tree/Demo_Tree_Large_Destroyed_Distant_Att}
 }
 {id=Delay_0, children [Modulator_5]}
 {id=Delay_1, children [Modulator_8], properties {delayMax=0.3, delayMin=0.3}}
 {id=Mixer_0, children [Delay_0,Modulator_0], properties {inputVolume [0.6,0.8]}}
 {id=Mixer_1, children [SoundClass_0,SoundClass_2], properties {inputVolume [1,0.4]}}
 {id=Modulator_0, children [Random_0], properties {pitchMax=1.3}}
 {id=Modulator_5, children [Random_2], properties {pitchMax=1.3, pitchMin=0.9, volumeMax=1, volumeMin=0.9}}
 {id=Modulator_7, children [Mixer_1], properties {pitchMax=1.1, pitchMin=1}}
 {id=Modulator_8, children [Random_7], properties {volumeMax=2, volumeMin=2}}
 {
  id=Random_0
  children [WavePlayer_5, WavePlayer_6, WavePlayer_9, WavePlayer_10, WavePlayer_11]
  properties {bShouldExcludeFromBranchCulling=true, weights [1,1,1,1,1]}
 }
 {id=Random_2, children [WavePlayer_7, WavePlayer_8, WavePlayer_14], properties {bShouldExcludeFromBranchCulling=true, weights [1,1,1]}}
 {id=Random_7, children [WavePlayer_22, WavePlayer_23, WavePlayer_24], properties {bShouldExcludeFromBranchCulling=true, weights [1,1,1]}}
 {
  id=SoundClass_0
  children [Attenuation_2]
  properties {soundClassOverride=&Fort_WeaponRanged_Fire_3P_PSM}
 }
 {
  id=SoundClass_2
  children [Attenuation_7]
  properties {soundClassOverride=&Fort_WeaponRanged_Fire_Distant}
 }
 {id=WavePlayer_10, properties {soundWaveAssetPtr=/Game/Sounds/Fort_Impact_Sounds/Demolition/Tree/Demo_Tree_Large_Destroyed_Break_04}}
 {id=WavePlayer_11, properties {soundWaveAssetPtr=/Game/Sounds/Fort_Impact_Sounds/Demolition/Tree/Demo_Tree_Large_Destroyed_Break_05}}
 {id=WavePlayer_14, properties {soundWaveAssetPtr=&/PhysicsTree_Destroyed_03}}
 {id=WavePlayer_22, properties {soundWaveAssetPtr=&/PhysicsTree_Destroyed_Distant_01}}
 {id=WavePlayer_23, properties {soundWaveAssetPtr=&/PhysicsTree_Destroyed_Distant_02}}
 {id=WavePlayer_24, properties {soundWaveAssetPtr=&/PhysicsTree_Destroyed_Distant_03}}
 {id=WavePlayer_5, properties {soundWaveAssetPtr=/Game/Sounds/Fort_Impact_Sounds/Demolition/Tree/Demo_Tree_Large_Destroyed_Break_01}}
 {id=WavePlayer_6, properties {soundWaveAssetPtr=/Game/Sounds/Fort_Impact_Sounds/Demolition/Tree/Demo_Tree_Large_Destroyed_Break_02}}
 {id=WavePlayer_7, properties {soundWaveAssetPtr=&/PhysicsTree_Destroyed_01}}
 {id=WavePlayer_8, properties {soundWaveAssetPtr=&/PhysicsTree_Destroyed_02}}
 {id=WavePlayer_9, properties {soundWaveAssetPtr=/Game/Sounds/Fort_Impact_Sounds/Demolition/Tree/Demo_Tree_Large_Destroyed_Break_03}}
]
properties {bDebug=true, bOverrideConcurrency=true, maxDistance=18000, volumeMultiplier=0.7, concurrencyOverrides {maxCount=4}}
